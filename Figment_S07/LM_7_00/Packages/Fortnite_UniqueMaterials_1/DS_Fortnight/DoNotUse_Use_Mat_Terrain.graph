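# Terrain master material, named as do-not-use, that layers grass, mud and cliff rock for a Fortnite-style landscape, with Nanite usage enabled.
# GrassTiling supplies grass, switchable to a world-projected replacement texture; an unconnected lerp would blend them by vertex color.
# HeightLerps add optional world-XY mud (green vertex channel by default), then world-aligned cliff rock (blue), using texture heights.
# The resulting color feeds a deprecated base shader function.
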Material /Figment_S07/LM_7_00/Packages/Fortnite_UniqueMaterials_1/DS_Fortnight/DoNotUse_Use_Mat_Terrain
graph {
 properties {bUsedWithNanite=true, pixelDepthOffsetMode=PDOM_Legacy, usedShadingModels=Default Lit}
 nodes [
  n1 = VectorParameter() {group=Code, parameterName=DiffuseColoration, defaultValue=<rgba 1 1 1 0>}
  AppendVector(A=@n1.RGB, B=@n1.RGB)
  Constant4Vector()
  n2 = TextureObjectParameter() {
   group=Grass
   parameterName=GrassReplacementTextureXYZ
   texture=/Figment_S07/LM_7_00/Packages/DS_Fortnite_Terrain_NoLOD/Textures/T_FORT_Underground_02
  }
  n3 = ScalarParameter() {defaultValue=512, group=Grass, parameterName=GrassReplacementWorldCoordScale}
  n4 = WorldPosition()
  n5 = TextureSampleParameter2D(UVs=ComponentMask(None=Divide(A=@n4.XYZ, B=@n3)) {g=true, r=true}) {
   group=Grass
   parameterName=GrassReplacementTexture
   texture=&T_FORT_Underground_02
  }
  n6 = StaticSwitchParameter(
   True=MaterialFunctionCall(TextureObject=@n2, TextureSize=@n3, WorldPosition=@n4.XYZ, Export Float 4=StaticBool() {value=true}) {materialFunction=&WorldAlignedTexture}.XYZ Texture
   False=AppendVector(A=@n5.RGB, B=@n5.RGB)
  ) {group=Grass, parameterName=UseXYZCoords}
  n7 = MaterialFunctionCall!(
   "Grass 01 SpecMask (Alpha Contains Grass 1 2 Lerp Mask)"
   Use Grass Normals
   Grass 01
   Grass 02 SpecMask
   Grass 02
   GrassNormal_01
   GrassNormal_02
  ) {materialFunction=&/GrassTiling}
  n8 = AppendVector(A=@n7.Diffuse, B=ComponentMask(None=BreakMaterialAttributes!(Attr=@n7.MaterialAttributes - fortnite format).Opacity) {r=true})
  n9 = VertexColor()
  LinearInterpolate(
   A=@n6
   B=@n8
   Alpha=StaticComponentMaskParameter(None=AppendVector(A=@n9, B=@n9)) {defaultR=true, group=GrassReplace, parameterName=GrassReplaceVertexColorChannel}
  )
  n10 = StaticSwitchParameter(True=@n6, False=@n8) {group=Grass, parameterName=ReplaceGrassTexture}
  n11 = ScalarParameter() {defaultValue=512, group=Mud, parameterName=MudWorldCoordScale}
  n12 = TextureSampleParameter2D(UVs=ComponentMask(None=Divide(A=@n4.XYZ, B=@n11)) {g=true, r=true}) {group=Mud, parameterName=MudTexture, texture=/Figment_S07/LM_7_00/Packages/DS_Fortnight/Mud}
  n13 = VertexColor()
  n14 = ComponentMask(None=@n12.A) {desc=mud Height map, r=true}
  n15 = Multiply(A=ScalarParameter() {defaultValue=0.5, group=Mud, parameterName=Amount of grass in mud}, B=ComponentMask(None=@n10) {a=true})
  n16 = MaterialFunctionCall(
   A=ComponentMask(None=@n10) {b=true, g=true, r=true}
   B=Multiply(A=@n12.RGB, B=ScalarParameter() {defaultValue=0.3, group=Mud, parameterName=MudMultiply})
   Transition Phase=StaticComponentMaskParameter(None=AppendVector(A=@n13, B=@n13)) {defaultG=true, group=Mud, parameterName=MudVertexColorChannel}
   Height Texture=Clamp(None=Add(A=@n14, B=@n15))
   Contrast=ScalarParameter() {defaultValue=2, group=Mud, parameterName=MudLerpContrast}
  ) {materialFunction=&HeightLerp}
  n17 = StaticSwitchParameter(True=@n16.Results, False=@n10) {defaultValue=true, group=Mud, parameterName=UseMudTexture}
  n18 = MaterialFunctionCall(
   TextureObject=TextureObjectParameter() {group=Cliffs, parameterName=CliffTexture, texture=/Figment_S07/LM_7_00/Packages/DS_Fortnight/SM/Textures/T_Cliff_Desert_Rock_2}
   TextureSize=ScalarParameter() {defaultValue=-768, group=Cliffs, parameterName=WorldCoordScale}
   Export Float 4=StaticBool() {value=true}
  ) {materialFunction=&WorldAlignedTexture}
  n19 = VertexColor()
  n20 = AppendVector(A=@n19, B=@n19)
  n21 = StaticComponentMaskParameter(None=@n18.XYZ Texture) {defaultR=true, parameterName=CliffHeightMapChannel}
  n22 = StaticSwitchParameter(True=@n16.Alpha, False=ComponentMask(None=@n10) {a=true}) {defaultValue=true, group=Mud, parameterName=UseMudTexture}
  n23 = ScalarParameter() {defaultValue=0.5, group=Cliffs, parameterName=Grass or Rock Lerp Alpha}
  n24 = MaterialFunctionCall!(
   -----------Base Shader-----------
   BaseColor1=MaterialFunctionCall(
    A=ComponentMask(None=@n17) {b=true, g=true, r=true}
    B=ComponentMask(None=@n18.XYZ Texture) {b=true, g=true, r=true}
    Transition Phase=StaticComponentMaskParameter(None=@n20) {defaultB=true, group=Cliffs, parameterName=CliffVertexColorMaskChannel}
    Height Texture=LinearInterpolate(A=@n21, B=@n22, Alpha=@n23)
    Contrast=ScalarParameter() {defaultValue=2, group=Cliffs, parameterName=CliffLerpContrast}
   ) {materialFunction=&HeightLerp}.Results
   Base Color Alpha
   SpecularMasks1
   Emissive
   Normals1
   World Position Offset
   DiffuseFresnelAmount=ScalarParameter() {defaultValue=0.25, parameterName=DiffuseFresnelAmount}
   -----------Destruction------------
   ------Multiple Diffuse Textures-----
   BaseColor2
   BaseColor3
   BaseColor4
   BaseColor5
   BaseColor6
   BaseColor7
   BaseColor8
   " - - - - - - - -"
   SpecularMasks2
   SpecularMasks3
   SpecularMasks4
   SpecularMasks5
   SpecularMasks6
   SpecularMasks7
   SpecularMasks8
   " - - - - - - - - "
   normal2
   Normal3
   Normal4
   Normal5
   Normal6
   Normal7
   Normal8
   " - - - - - - - - 1"
   Use 2 Textures
   Use 3 Textures
   Use 4 Textures
   Use 5 Textures
   Use 6 Textures
   Use 7 Textures
   Use 8 Textures
   Use 9 Textures
   Use 10 Textures
   ExposeLerpFunctionForInvertingTheHitDistanceCalc
  ) {materialFunction=/Figment_S07/LM_7_00/Packages/Fortress_SharedMaterials/Base_Material_Functions/DEPRECATED_BaseShaderFunction}
  MaterialFunctionCall!(Power=@n24.Roughness) {materialFunction=/Engine/Functions/Engine_MaterialFunctions01/Shading/PowerToRoughness}
  Multiply(A=@n24.Specular) {constA=1, constB=0.04, desc=FresnelBaseReflectFraction}
  OneMinus()
  ScalarParameter() {group=Code, parameterName=Code_DestructionAmount}
  ScalarParameter() {group=Code, parameterName=WSDamageRadius}
  ScalarParameter() {group=Code, parameterName=Code_HitAnimation}
  StaticBool()
  VectorParameter() {parameterName=Code_DeformationVector}
  VectorParameter() {parameterName=Code_DeformationCenter}
 ]
}
